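# Translucent unlit Niagara sprite material that draws the stopwatch emote's digital timer as a hologram panel.
# MF_DigitalTimer masks colored digits over a gradient or particle-color background, with slant and scale UV math declared for skewing.
# MF_3DHB_HologramFX blends scanline color into the digits and modulates opacity alongside a round or soft-edged rectangular mask and depth fade.
# Dynamic parameters Erosion and ErosionPan drive a panning, circuit-textured erosion mask.
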Material /BRCosmetics/Animation/Game/MainPlayer/Emotes/StopWatch/FX/M_Emote_Stopwatch_Digits
graph {
 properties {
  bAutomaticallySetUsageInEditor=false
  bUsedWithNiagaraSprites=true
  blendMode=BLEND_Translucent
  shadingModel=MSM_Unlit
  usedShadingModels=Unlit
 }
 nodes [
  n1 = VectorParameter() {group=Background, parameterName=BackgroundColor_Black, defaultValue=<rgba 0.5735 0.1235 0 1>}
  n2 = VectorParameter() {group=Background, parameterName=BackgroundColor_White, defaultValue=<rgba 1 0.5318 0.1534 1>}
  n3 = MaterialFunctionCall(
   UVs=NamedRerouteUsage()
   Radius=ScalarParameter() {defaultValue=0.7, group=Background, parameterName=BackgroundColor_Radius}
   Density=ScalarParameter() {defaultValue=2, group=Background, parameterName=BackgroundColor_Density}
  ) {materialFunction=&RadialGradientExponential}
  n4 = VectorParameter() {group=DigitalTimer, parameterName=DigitsColor, defaultValue=<rgba 0 0 0 1>}
  n5 = Constant() {r=1}
  n6 = LinearInterpolate(
   A=VectorParameter() {group=DigitalTimer, parameterName=DigitsColorOff, defaultValue=<rgba 0.0214 0 0.2 1>}.RGB
   B=VectorParameter() {group=DigitalTimer, parameterName=DigitsColorOn, defaultValue=<rgba 0.1864 0.1234 1 1>}.RGB
   Alpha=If(A=NamedRerouteUsage(), "A > B"=@n5, "A == B"=@n5, "A < B"=Constant()) {constB=0.18}
  )
  n7 = ScalarParameter() {parameterName=UseSeparateOnOffColors}
  n8 = MaterialFunctionCall!(
   Time=Multiply(A=Time(), B=ScalarParameter() {defaultValue=1, group=Background, parameterName=HologramSpeed})
   MeshScaleIn
   UVForScanline
   CustomFlicker
   UseCustomFlicker
   UVForStroke
   UVForGlitch
   BaseUVIn
   EmissiveIn
   OpacityIn
   WPOIn
  ) {materialFunction=/Game/Tools/3DHB/Shader/Base/MF_3DHB_HologramFX}
  n9 = MaterialFunctionCall!(UVs=NamedRerouteUsage()) {materialFunction=/BRCosmetics/Animation/Props/Gadgets_Stopwatch/Materials/MF_DigitalTimer}
  LinearInterpolate(
   A=StaticSwitchParameter(True=ParticleColor().RGB, False=LinearInterpolate(A=@n1.RGB, B=@n2.RGB, Alpha=@n3.RadialGradientExponential)) {group=Background, parameterName=UseParticleColorBackground}
   B=LinearInterpolate(
    A=LinearInterpolate(A=@n4.RGB, B=@n6, Alpha=@n7)
    B=VectorParameter() {group=BASE_HOLOFX_SCANLINE, parameterName=ScanlineColor, defaultValue=<rgba 1 0.2075 0 1>}.RGB
    Alpha=@n8.EmissiveOut
   )
   Alpha=@n9.Result
  )
  n10 = NamedRerouteUsage()
  n11 = ScalarParameter() {defaultValue=0.6, group=Background, parameterName=Background_Radius}
  n12 = ScalarParameter() {defaultValue=2.33, group=Background, parameterName=Background_Density}
  n13 = Saturate(None=NamedRerouteUsage())
  n14 = ComponentMask(None=@n13) {r=true}
  n15 = ComponentMask(None=@n13) {g=true}
  n16 = Multiply(A=Multiply(A=OneMinus(None=@n14), B=@n14), B=Multiply(A=OneMinus(None=@n15), B=@n15))
  n17 = ScalarParameter() {defaultValue=0.03, group=Background, parameterName=BackgroundEdgeStrength}
  n18 = Multiply(
   A=@n8.OpacityOut
   B=StaticSwitchParameter(
    True=MaterialFunctionCall(UVs=@n10, Radius=@n11, Density=@n12) {materialFunction=&RadialGradientExponential}.RadialGradientExponential
    False=MaterialFunctionCall(Input=@n16, Input High=@n17) {materialFunction=&RemapValueRange}.Result
   ) {group=Background, parameterName=RoundBackground}
  )
  n19 = StaticSwitchParameter(True=ParticleColor().A, False=ScalarParameter() {defaultValue=2, parameterName=Opacity, sortPriority=2}) {parameterName=UseParticleColorOpacity, sortPriority=1}
  Multiply(
   A=DepthFade(Opacity=Multiply(A=@n18, B=@n19) {constB=2}, FadeDistance=ScalarParameter() {defaultValue=25, parameterName=DepthFade})
   B=NamedRerouteUsage()
  )
  n20 = ScalarParameter() {defaultValue=-0.2, parameterName=SlantAmount}
  n21 = Multiply(A=@n20) {constB=-0.5}
  n22 = Constant()
  n23 = MaterialFunctionCall(UVs=Add(A=TextureCoordinate(), B=AppendVector(A=@n21, B=@n22)), Texture Scale=ScalarParameter() {defaultValue=0.8, parameterName=Scale}) {materialFunction=&ScaleUVsByCenter}
  n24 = ComponentMask(None=@n23.UVs) {r=true}
  n25 = ComponentMask(None=@n23.UVs) {g=true}
  n26 = Multiply(A=OneMinus(None=@n25), B=@n20)
  NamedRerouteDeclaration(None=AppendVector(A=Add(A=@n24, B=@n26), B=@n25)) {name=SlantUVs, nodeColor=<rgba 1 0 0.1172 1>}
  n27 = TextureCoordinate()
  n28 = ComponentMask(None=@n27) {r=true}
  n29 = OneMinus(None=@n28)
  n30 = Saturate(None=Multiply(A=Multiply(A=@n28, B=@n29)) {constB=2})
  n31 = DynamicParameter() {defaultValue=<rgba 0 1 1 1>, paramNames [Erosion,ErosionPan,Param3,Param4]}
  n32 = Multiply(A=@n31.Erosion) {constB=10}
  n33 = Add(A=@n27, B=AppendVector(A=@n31.ErosionPan, B=Constant()))
  n34 = Saturate(None=Add(A=TextureSampleParameter2D(UVs=@n33) {group=Erosion, parameterName=ErosionTexture, texture=/Game/Effects/Niagara/MasterMaterials/Tech/Textures/T_Circuits_Diffuse}.B) {constB=-0.1})
  NamedRerouteDeclaration(None=MaterialFunctionCall(Alpha=Power(Base=@n30, Exp=@n32) {constExponent=0.5}, Max=OneMinus(None=@n34)) {materialFunction=&SmoothStep}.Result) {name=Erosion, nodeColor=<rgba 0.7344 1 0 1>}
  NamedRerouteDeclaration(None=@n9.Result) {name=Digital Timer, nodeColor=<rgba 0.5234 1 0 1>}
 ]
}
