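# Niagara system for the Stopwatch emote's digit display, drawing sprites from three timed bursts that persist past their lifetime.
# Emitter-age curves drive a squash-and-stretch pop-in and dynamic material parameters.
# A User.TimerStart threshold check latches the engine time once, from which Emitter.Timer computes elapsed seconds.
# The shown stack never reads Emitter.Timer or the countdown, count-up and delay user variables.
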NiagaraSystem /BRCosmetics/Animation/Game/MainPlayer/Emotes/StopWatch/FX/NS_Emote_Stopwatch_Digits
dependencies {
 usedDataInterfaces [&NiagaraDataInterfaceCurve, &NiagaraDataInterfaceVector2DCurve]
 usedDynamicInputs []
 usedModules [&SystemState]
}
emitters {Digits {
  emitterSpawnScript [/SetVariables() {isSetParameters=true}]
  emitterUpdateScript [
   &EmitterState(Life Cycle Mode=::NewEnumerator1, Loop Behavior=::NewEnumerator1, Loop Duration Mode=::NewEnumerator1, UseLoopDelay=false)
   &SpawnBurst_Instantaneous(Use Spawn Probability=false, UseLoopCountLimit=false)
   &SpawnBurst_Instantaneous(Spawn Time=3.2, Use Spawn Probability=false, UseLoopCountLimit=false) {module=SpawnBurst_Instantaneous001}
   &SpawnBurst_Instantaneous(Spawn Time=1.15, Use Spawn Probability=false, UseLoopCountLimit=false) {module=SpawnBurst_Instantaneous002}
   /SetVariables(Emitter.StartTime:NiagaraFloat=&MakeCustomFloatFromBool(
     Bool=&Boolean_AndOperation(
      Bool1=&Boolean_NotOperation(Bool=$Emitter.HasTimerStarted)
      Bool2=&SetBoolByFloatComparison(A=$User.TimerStart, B=0.05)
     )
     True Float=$Engine.Time
     False Float=$Emitter.StartTime
    )) {isSetParameters=true}
   /SetVariables(Emitter.HasTimerStarted:NiagaraBool=&Boolean_OrOperation(Bool1=$Emitter.HasTimerStarted, Bool2=&SetBoolByFloatComparison(A=$User.TimerStart, B=0.05))) {isSetParameters=true}
   /SetVariables(Emitter.Timer:NiagaraFloat=&MakeCustomFloatFromBool(
     Bool=$Emitter.HasTimerStarted
     True Float=&Max_Float(A:NiagaraFloat=&Subtract_Float(A=$Engine.Time, B=$Emitter.StartTime), B:NiagaraFloat=0)
     False Float=0
    )) {isSetParameters=true}
  ]
  particleSpawnScript [&InitializeParticle(
    Color Mode=::NewEnumerator0
    Lifetime=0
    Sprite Size=&Multiply_Vector2D_ByFloat(Float=70, Vector2D=<1 0.4>)
    Sprite Size Mode=::NewEnumerator4
    UsePositionOffset=false
    Write Material Random=true
   )]
  particleUpdateScript [
   &ParticleState(Kill Particles When Lifetime Has Elapsed=false)
   &ScaleSpriteSize(
    Non-Uniform Curve Index:NiagaraFloat=$Emitter.Age
    Non-Uniform Curve Sprite Scale=DataInterface() {propertyValues {
      XCurve {keys [{tangentMode=RCTM_User, time=0, value=0}, <at 0.0648 0.1731>, <at 0.1751 1.338>, <at 0.25 1>]}
      YCurve {keys [{tangentMode=RCTM_User, time=0, value=0}, <at 0.0878 1.4818>, <at 0.1593 0.8642>, <at 0.25 1>]}
     }}
    Scale Sprite Size Mode=::NewEnumerator3
   )
   &DynamicMaterialParameters(
    Index 0 Param 1=&FloatFromCurve(
     FloatCurve=DataInterface() {propertyValues {Curve {keys [
         {arriveTangent=-4.1758, interpMode=RCIM_Cubic, leaveTangent=-4.1758, tangentMode=RCTM_User, time=0, value=1}
         {arriveTangent=-1.1026, interpMode=RCIM_Cubic, leaveTangent=-1.1026, tangentMode=RCTM_User, time=0.35, value=0}
        ]}}}
     CurveIndex=$Emitter.Age
    )
    Index 0 Param 2=&FloatFromCurve(FloatCurve=DataInterface() {propertyValues {Curve {keys [<at 0 1>,<at 1 0>]}}}, CurveIndex=$Emitter.Age, Scale Curve=20)
    Module.Index0Float0Write=true
    Module.Index0Float1Write=true
    Module.Index0Float2Write=true
    Module.Index0Float3Write=true
   )
  ]
  rendererClasses [&NiagaraSpriteRendererProperties]
  renderers [{rendererClass=&NiagaraSpriteRendererProperties}]
 }}
summary {
 systemName=NS_Emote_Stopwatch_Digits
 emitters [{emitterName=Digits, rendererClasses [&NiagaraSpriteRendererProperties]}]
 userVariables [
  User.CountdownTime {type=NiagaraFloat, defaultValue=10, description=""}
  User.CountUp {type=NiagaraBool, defaultValue=true}
  User.TimerDelay {type=NiagaraFloat, defaultValue=1.32, description=""}
  User.TimerStart {type=NiagaraFloat, defaultValue=0, description=""}
 ]
}
